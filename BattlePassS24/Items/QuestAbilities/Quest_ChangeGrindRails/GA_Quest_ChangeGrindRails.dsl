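; Season 24 battle pass quest ability that detects a player jumping from a grind rail back onto a rail.
; Chained tag waits track grinding ending, then fall-damage immunity being added and removed, before rechecking the grinding tag.
; Success only prints a debug message, while failure prints and ends the ability.
; ShouldAbilityRespondToEvent accepts only critical hits on non-teammate player pawns or other FortPawns, a filter unrelated to grinding.
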(event Ability|EventActivateAbilityFromEvent
  (params (EventData :type Gameplay%20Event%20Data%20Structure%20%28by%20ref%29))
  (|CallParentFunction EventData)
  (Development|PrintString "@@@ ActivateAbilityFromEvent YOU ARE GRINDING!!!!" true false "(R=1.000000,G=0.000000,B=0.566126,A=1.000000)" 5.0)
  (Ability|Tasks|WaitGameplayTagRemove "(TagName=\"Granted.Athena.GrindRails.IsGrinding\")"
    (:Removed
      (Development|PrintString "YOU JUMPED OFF RAIL." true true "(R=0.000000,G=0.012131,B=0.755208,A=1.000000)" 6.0)
      (Ability|Tasks|WaitGameplayTagAdd "(TagName=\"Gameplay.Status.ImmuneFallingDamage\")"
        (:Added
          (Development|PrintString "YOU ARE IN THE AIR!!" true true "(R=0.000000,G=0.660000,B=1.000000,A=1.000000)" 4.0)
          (Ability|Tasks|WaitGameplayTagRemove "(TagName=\"Gameplay.Status.ImmuneFallingDamage\")"
            (:Removed
              (if (GameplayTags|HasTag (GameplayTags|GetOwnedGameplayTags (Variables|GAVars|Data|GetOwningPlayerPawn) :!class /Script/GameplayTags.GameplayTagAssetInterface) "(TagName=\"Granted.Athena.GrindRails.IsGrinding\")")
                (Development|PrintString "PASSED!! YOU ARE GRINDING!!" true true "(R=0.028321,G=1.000000,B=0.000000,A=1.000000)" 5.0)
                (else
                  (Development|PrintString "FAILED. You are no longer grinding." true true "(R=1.000000,G=0.042986,B=0.000000,A=1.000000)" 5.0)
                  (Ability|EndAbility))))))))))

(fn ShouldAbilityRespondToEvent
  (params (ActorInfo :type Gameplay%20Ability%20Actor%20Info%20Structure) (Payload_EventTag :type Gameplay%20Tag%20Structure) (Payload_Instigator :type Actor%20Object%20Reference) (Payload_Target :type Actor%20Object%20Reference) (Payload_OptionalObject :type Object%20Reference) (Payload_OptionalObject2 :type Object%20Reference) (Payload_ContextHandle :type Gameplay%20Effect%20Context%20Handle%20Structure) (Payload_InstigatorTags :type Gameplay%20Tag%20Container%20Structure) (Payload_TargetTags :type Gameplay%20Tag%20Container%20Structure) (Payload_EventMagnitude :type Float%20%28single-precision%29) (Payload_TargetData :type Gameplay%20Ability%20Target%20Data%20Handle%20Structure) (Payload_InstancedEventData :type Instanced%20Struct%20Structure))
  (bind _returnvalue (Ability|EffectContext|IsCriticalHit Payload_ContextHandle))
  (bind _asfort_player_pawn (Utilities|Casting|CastToFortPlayerPawn Payload_Target)
    (:then
      (return (and (not (Team|OnSameTeam _asfort_player_pawn Payload_Instigator)) _returnvalue)))
    (:CastFailed
      (Utilities|Casting|CastToFortPawn Payload_Target
        (:then
          (if _returnvalue
            (return true)
            (else
              (return false))))
        (:CastFailed
          (return false))))))
